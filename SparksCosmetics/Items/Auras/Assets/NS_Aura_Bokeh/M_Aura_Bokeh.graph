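# Translucent unlit Niagara sprite material that draws soft bokeh discs for the SparksCosmetics aura effect.
# Exponential radial gradients form a disc and rim, with dynamic parameters for ring thickness, focus softness and a randomized offset overlap spot.
# That shape, scaled by particle alpha, feeds a depth fade.
# Gradient-shaded particle color feeds emissive adjustments directly, while a separate hue-shift branch does not feed them.
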Material /SparksCosmetics/Items/Auras/Assets/NS_Aura_Bokeh/M_Aura_Bokeh
graph {
 properties {
  bAutomaticallySetUsageInEditor=false
  bUsedWithNiagaraSprites=true
  blendMode=BLEND_Translucent
  shadingModel=MSM_Unlit
  usedShadingModels=Unlit
 }
 nodes [
  Add()
  n1 = ParticleColor()
  n2 = Constant2Vector() {g=0.5, r=0.5}
  n3 = Constant() {r=0.4}
  n4 = DynamicParameter() {defaultValue=<rgba 0.385 1 1 1>, paramNames [Ring Thickness, Focus, Overlap On/Off, Param4]}
  n5 = LinearInterpolate(A=Constant() {r=2}, B=Constant() {r=25}, Alpha=@n4.Focus) {constAlpha=0}
  n6 = LinearInterpolate(A=Constant() {r=2}, B=Constant() {r=50}, Alpha=@n4.Focus) {constAlpha=0}
  n7 = MaterialFunctionCall(CenterPosition=@n2, Radius=@n3, Density=@n6) {materialFunction=&RadialGradientExponential}
  n8 = MaterialFunctionCall(CenterPosition=@n2, Radius=@n4.Ring Thickness, Density=@n6) {materialFunction=&RadialGradientExponential}
  n9 = Add(
   A=Multiply(A=MaterialFunctionCall(CenterPosition=@n2, Radius=@n3, Density=@n5) {materialFunction=&RadialGradientExponential}.RadialGradientExponential) {constB=0.3}
   B=Saturate(None=Subtract(A=@n7.RadialGradientExponential, B=@n8.RadialGradientExponential))
  )
  n10 = Constant2Vector() {g=0.45, r=0.55}
  n11 = Add(A=@n5, B=Multiply(A=ParticleRandom()) {constB=-0.2})
  n12 = Saturate(None=Add(
    A=@n9
    B=MaterialFunctionCall(CenterPosition=@n10, Radius=@n3, Density=@n11) {materialFunction=&RadialGradientExponential}.RadialGradientExponential
   ))
  DepthFade(
   Opacity=Multiply(A=@n1.A, B=LinearInterpolate(A=@n9, B=@n12, Alpha=@n4.Overlap On/Off))
   FadeDistance=ScalarParameter() {defaultValue=15, parameterName=DF Distance}
  )
  n13 = Multiply(A=Time(), B=ScalarParameter() {defaultValue=0.1, parameterName=HueShift Speed}) {constB=0.1}
  n14 = Multiply(A=Abs(None=Sine(None=@n13))) {constB=0}
  n15 = ScalarParameter() {defaultValue=0.1, parameterName=Fixed HueShift Percentage}
  n16 = ScalarParameter() {parameterName=Fixed Hueshift On/Off}
  n17 = Constant2Vector() {g=0.4, r=0.6}
  n18 = Add(A=OneMinus(None=MaterialFunctionCall(CenterPosition=@n17) {materialFunction=&RadialGradientExponential}.RadialGradientExponential)) {constB=0.85}
  n19 = Multiply(A=@n18, B=@n1.RGB)
  LinearInterpolate(
   A=MaterialFunctionCall(Hue Shift Percentage=LinearInterpolate(A=@n14, B=@n15, Alpha=@n16), Texture=@n19) {materialFunction=&HueShift}.Result
   B=@n19
   Alpha=@n18
  )
  MaterialFunctionCall(Emissive=@n19) {materialFunction=&MF_FX_EmissiveAdjustments}
  n20 = TextureCoordinate()
  Saturate(None=Add(A=ComponentMask(None=@n20) {r=true}) {constB=0.3})
 ]
}
