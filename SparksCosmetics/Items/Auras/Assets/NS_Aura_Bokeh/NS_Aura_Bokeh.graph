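# Niagara aura cosmetic that surrounds a character's skeletal mesh with colored bokeh lights, glows, dust and rays.
# Two bokeh emitters flicker and shift focus inside a bone-anchored sphere, trailing it through velocity-scaled position interpolation.
# Glow sprites pin to seven bones or sockets, rising dust spawns at two, and a ray mesh cycles orange, blue and green.
# Time-of-day scalars and front-end and large-body flags adjust brightness, opacity and size.
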NiagaraSystem /SparksCosmetics/Items/Auras/Assets/NS_Aura_Bokeh/NS_Aura_Bokeh
dependencies {
 usedDataInterfaces [
  &NiagaraDataInterfaceArrayFloat
  &NiagaraDataInterfaceArrayInt32
  &NiagaraDataInterfaceCurve
  &NiagaraDataInterfaceArrayColor
  &NiagaraDataInterfaceArrayFloat3
  &NiagaraDataInterfaceColorCurve
  &NiagaraDataInterfaceSkeletalMesh
 ]
 usedDynamicInputs []
 usedModules [
  &self:SystemSpawnScript.SystemScriptSource.SystemScriptGraph.NiagaraNodeAssignment_0.SetVariables_#1
  &SystemState
 ]
}
emitters {
 BokehLights {
  emitterUpdateScript [
   &EmitterState()
   &SpawnRate(
    SpawnRate=&Multiply_Float(A=5, B=&MakeCustomFloatFromBool(Bool=$User.IsLargeBody, True Float=1, False Float=1))
    Use Spawn Probability=false
   )
  ]
  particleSpawnScript [
   &InitializeParticle(
    Color=&Multiply_LinearColorByFloat(
     Linear Color=&SelectLinearColorFromArray(
      Array Sampling Mode=::NewEnumerator0
      Direct Array Index=?
      Color Selection Array=DataInterface() {propertyValues {ColorData [<rgba 1 0.4392 0 1>, <rgba 0 1 0.3399 1>, <rgba 0 0.3405 1 1>]}}
      Random Seed=0
     )
     Float=&FNTimeOfDayScalar(Night Scale=0.9, Morning Scale=1, Evening Scale=1)
    )
    Lifetime=&RandomRangeFloat(Minimum=2, Maximum=3)
    Sprite Rotation Mode=::NewEnumerator2
    Sprite Size=&Multiply_Vector2D_ByFloat(
     Float=&MakeCustomFloatFromBool(Bool=$User.IsLargeBody, True Float=1, False Float=1)
     Vector2D=&Multiply_Vector2D_ByFloat(
      Float=&RandomRangeFloat(Minimum=2, Maximum=3)
      Vector2D=&RandomRangeVector2D(Minimum=<6.5 6.5>, Maximum=<7 7>)
     )
    )
    Sprite Size Mode=::NewEnumerator4
    UsePositionOffset=false
    Write Material Random=true
   ) {module=InitializeParticle001}
   /SetVariables(
    Particles.FlickerMask:NiagaraInt32=&RandomRangeInt(Minimum=0, Maximum=1)
    Particles.FocusMask:NiagaraInt32=&RandomRangeInt(Minimum=0, Maximum=1)
   ) {isSetParameters=true, module=SetVariables_#2}
   &DynamicMaterialParameters(
    Index 0 Param 3=&SelectFloatFromArray(
     Array Sampling Mode=::NewEnumerator0
     Direct Array Index=?
     Float Selection Array=DataInterface() {propertyValues {FloatData [0,0,1]}}
     Random Seed=0
    )
    Module.Index0Float0Write=true
    Module.Index0Float1Write=true
    Module.Index0Float2Write=true
    Module.Index0Float3Write=false
   ) {module=DynamicMaterialParameters001}
  ]
  particleUpdateScript [
   &ParticleState()
   &ScaleSpriteSize(Uniform Curve Sprite Scale=DataInterface() {propertyValues {Curve {keys [
        {arriveTangent=0.5547, interpMode=RCIM_Cubic, leaveTangent=0.5547, tangentMode=RCTM_User, time=0, value=0.9}
        {interpMode=RCIM_Cubic, tangentMode=RCTM_User, time=0.5, value=1}
        {arriveTangent=0.0003, tangentMode=RCTM_User, time=1, value=0.9}
       ]}}})
   &SkeletalMeshLocation(
    Bone / Socket Index=7
    Bone Sampling Mode=::NewEnumerator4
    MaskSpawnGroup=false
    Orientation Sampling=::NewEnumerator1
    Skeletal Mesh=$System.Skeletal Mesh
   ) {module=SkeletalMeshLocation001}
   &ShapeLocation(
    MaskSpawnGroup=false
    Offset Mode=::NewEnumerator0
    Override Seed=false
    Scale Mode:ENiagara_ScaleMode=::NewEnumerator1
    Shape Origin=$StackContext.Position
    Sphere Radius=&Multiply_Float(A=40, B=&MakeCustomFloatFromBool(Bool=$User.IsLargeBody, True Float=1.1, False Float=1))
   ) {module=ShapeLocation001}
   &DynamicMaterialParameters(
    Index 0 Param 1=&SelectFloatFromArray(
     Array Sampling Mode=::NewEnumerator0
     Direct Array Index=?
     Float Selection Array=DataInterface() {propertyValues {FloatData [0.38,0.4]}}
     Random Seed=0
    )
    Index 0 Param 2=&MakeCustomFloatFromBool(
     Bool=&SetBoolByIntComparison(A=$Particles.FocusMask, B=0, Comparison Type=::NewEnumerator0)
     True Float=&FloatFromCurve(FloatCurve=DataInterface() {propertyValues {Curve {keys [
          {arriveTangent=4.9984, interpMode=RCIM_Cubic, leaveTangent=4.9984, tangentMode=RCTM_User, time=0, value=0}
          {interpMode=RCIM_Cubic, tangentMode=RCTM_User, time=0.5, value=1}
          {tangentMode=RCTM_User, time=1, value=0}
         ]}}})
     False Float=&FloatFromCurve(FloatCurve=DataInterface() {propertyValues {Curve {keys [
          {arriveTangent=-5.3163, interpMode=RCIM_Cubic, leaveTangent=-5.3163, tangentMode=RCTM_User, time=0, value=1}
          {interpMode=RCIM_Cubic, tangentMode=RCTM_User, time=0.5, value=0}
          {arriveTangent=4.6849, tangentMode=RCTM_User, time=1, value=1}
         ]}}})
    )
    Index 0 Param 3=1
    Module.Param0WriteEnabled=true
    Module.Param1WriteEnabled=true
    Module.Param2WriteEnabled=false
    Module.Param3WriteEnabled=true
   )
   &ScaleColor(
    Scale Alpha=&FloatFromCurve(
     FloatCurve=DataInterface() {propertyValues {Curve {keys [
         {arriveTangent=5.4362, interpMode=RCIM_Cubic, leaveTangent=5.4362, tangentMode=RCTM_User, time=0, value=0}
         {interpMode=RCIM_Cubic, tangentMode=RCTM_User, time=0.5, value=1}
         {arriveTangent=-4.7327, tangentMode=RCTM_User, time=1, value=0}
        ]}}}
     Scale Curve=&RandomRangeFloat(Minimum=0.8, Maximum=1)
    )
    Scale RGB=&SelectVectorFromArray(
     Array Sampling Mode=::NewEnumerator0
     Direct Array Index=?
     Vector Selection Array=DataInterface() {propertyValues {FloatData [<1 1 1>, <1 1 1>, <2 2 2>]}}
     Random Seed=0
    )
    ScaleA=true
    ScaleRGB=true
   )
   &CameraOffset(Camera Offset Amount=-100) {module=CameraOffset001}
   &ScaleColor(
    Scale Alpha=&MakeCustomFloatFromBool(
     Bool=&SetBoolByIntComparison(A=$Particles.FlickerMask, B=0, Comparison Type=::NewEnumerator0)
     True Float=&Waveform(
      "Waveform [1]"=::NewEnumerator0
      "[1] Frequency"=&RandomRangeFloat(Minimum=1, Maximum=1.5)
      "[1] Amplitude Min/Max"=<0.5 1>
     )
     False Float=1
    )
    Scale RGB=&VectorFromFloat(Value=&MakeCustomFloatFromBool(
      Bool=&SetBoolByIntComparison(A=$Particles.FlickerMask, B=0, Comparison Type=::NewEnumerator0)
      True Float=&Waveform(
       "Waveform [1]"=::NewEnumerator0
       "[1] Frequency"=&RandomRangeFloat(Minimum=1.5, Maximum=2)
       "[1] Amplitude Min/Max"=<1 2>
      )
      False Float=1
     ))
    ScaleA=true
    ScaleRGB=true
   ) {module=ScaleColor001}
   &InterpolateOverTime(
    Rate Of Change:NiagaraFloat=&RemapRange(
     Input Value To Remap=&VectorLength(Vector=$Particles.SkeletalMeshLocation001.SampledVelocity)
     Input Min=0
     Input Max=770
     Output Min=1
     Output Max=30
    )
    Target Position Value:NiagaraPosition=$Particles.ShapeLocation001.ShapePosition
    Variable Type=::NewEnumerator6
   ) {module=InterpolateOverTime001}
   /SetVariables(Particles.Position=$Particles.InterpolateOverTime001.InterpolateOverTime_Position.Moving Average) {isSetParameters=true}
  ]
  rendererClasses [&NiagaraSpriteRendererProperties]
  renderers [{rendererClass=&NiagaraSpriteRendererProperties}]
 }
 BokehLights001 {
  emitterUpdateScript [
   &EmitterState()
   &SpawnRate(
    SpawnRate=&Multiply_Float(A=5, B=&MakeCustomFloatFromBool(Bool=$User.IsLargeBody, True Float=1, False Float=1))
    Use Spawn Probability=false
   )
  ]
  particleSpawnScript [
   &InitializeParticle(
    Color=&Multiply_LinearColorByFloat(
     Linear Color=&SelectLinearColorFromArray(
      Array Sampling Mode=::NewEnumerator0
      Direct Array Index=?
      Color Selection Array=DataInterface() {propertyValues {ColorData [<rgba 1 0.4392 0 1>, <rgba 0 1 0.3399 1>, <rgba 0 0.3405 1 1>]}}
      Random Seed=0
     )
     Float=&FNTimeOfDayScalar(Night Scale=0.9, Morning Scale=1, Evening Scale=1)
    )
    Lifetime=&RandomRangeFloat(Minimum=2, Maximum=3)
    Sprite Rotation Mode=::NewEnumerator2
    Sprite Size=&Multiply_Vector2D_ByFloat(
     Float=&MakeCustomFloatFromBool(Bool=$User.IsLargeBody, True Float=1, False Float=1)
     Vector2D=&Multiply_Vector2D_ByFloat(
      Float=&RandomRangeFloat(Minimum=3, Maximum=4)
      Vector2D=&RandomRangeVector2D(Minimum=<6.5 6.5>, Maximum=<7 7>)
     )
    )
    Sprite Size Mode=::NewEnumerator4
    UsePositionOffset=false
    Write Material Random=true
   ) {module=InitializeParticle001}
   /SetVariables(
    Particles.FlickerMask:NiagaraInt32=&RandomRangeInt(Minimum=0, Maximum=1)
    Particles.FocusMask:NiagaraInt32=&RandomRangeInt(Minimum=0, Maximum=1)
   ) {isSetParameters=true, module=SetVariables_#2}
   &DynamicMaterialParameters(
    Index 0 Param 3=&SelectFloatFromArray(
     Array Sampling Mode=::NewEnumerator0
     Direct Array Index=?
     Float Selection Array=DataInterface() {propertyValues {FloatData [0]}}
     Random Seed=0
    )
    Module.Index0Float0Write=true
    Module.Index0Float1Write=true
    Module.Index0Float2Write=true
    Module.Index0Float3Write=false
   ) {module=DynamicMaterialParameters001}
  ]
  particleUpdateScript [
   &ParticleState()
   &ScaleSpriteSize(Uniform Curve Sprite Scale=DataInterface() {propertyValues {Curve {keys [
        {arriveTangent=0.5547, interpMode=RCIM_Cubic, leaveTangent=0.5547, tangentMode=RCTM_User, time=0, value=0.9}
        {interpMode=RCIM_Cubic, tangentMode=RCTM_User, time=0.5, value=1}
        {arriveTangent=0.0003, tangentMode=RCTM_User, time=1, value=0.9}
       ]}}})
   &SkeletalMeshLocation(
    Bone / Socket Index=7
    Bone Sampling Mode=::NewEnumerator4
    MaskSpawnGroup=false
    Orientation Sampling=::NewEnumerator1
    Skeletal Mesh=$System.Skeletal Mesh
   ) {module=SkeletalMeshLocation001}
   &ShapeLocation(
    MaskSpawnGroup=false
    Non Uniform Scale=<1 1 0.8>
    Offset Mode=::NewEnumerator0
    Override Seed=false
    Rotation Mode=::NewEnumerator5
    Shape Origin=$StackContext.Position
    Sphere Radius=&Multiply_Float(A=40, B=&MakeCustomFloatFromBool(Bool=$User.IsLargeBody, True Float=1.1, False Float=1))
    UseOwnerScale=true
   ) {module=ShapeLocation001}
   &DynamicMaterialParameters(
    Index 0 Param 1=&SelectFloatFromArray(
     Array Sampling Mode=::NewEnumerator0
     Direct Array Index=?
     Float Selection Array=DataInterface() {propertyValues {FloatData [0.39,0.4]}}
     Random Seed=0
    )
    Index 0 Param 2=&MakeCustomFloatFromBool(
     Bool=&SetBoolByIntComparison(A=$Particles.FocusMask, B=0, Comparison Type=::NewEnumerator0)
     True Float=&FloatFromCurve(FloatCurve=DataInterface() {propertyValues {Curve {keys [
          {arriveTangent=2.0676, interpMode=RCIM_Cubic, leaveTangent=2.0676, tangentMode=RCTM_User, time=0, value=0}
          {interpMode=RCIM_Cubic, tangentMode=RCTM_User, time=0.5, value=1}
          {tangentMode=RCTM_User, time=1, value=0}
         ]}}})
     False Float=&FloatFromCurve(FloatCurve=DataInterface() {propertyValues {Curve {keys [
          {arriveTangent=-2.8705, interpMode=RCIM_Cubic, leaveTangent=-2.8705, tangentMode=RCTM_User, time=0, value=1}
          {interpMode=RCIM_Cubic, tangentMode=RCTM_User, time=0.5, value=0}
          {arriveTangent=3.8879, tangentMode=RCTM_User, time=1, value=1}
         ]}}})
    )
    Index 0 Param 3=1
    Module.Param0WriteEnabled=true
    Module.Param1WriteEnabled=true
    Module.Param2WriteEnabled=false
    Module.Param3WriteEnabled=true
   )
   &ScaleColor(
    Scale Alpha=&FloatFromCurve(
     FloatCurve=DataInterface() {propertyValues {Curve {keys [
         {arriveTangent=5.4362, interpMode=RCIM_Cubic, leaveTangent=5.4362, tangentMode=RCTM_User, time=0, value=0}
         {interpMode=RCIM_Cubic, tangentMode=RCTM_User, time=0.5, value=1}
         {arriveTangent=-4.7327, tangentMode=RCTM_User, time=1, value=0}
        ]}}}
     Scale Curve=&RandomRangeFloat(Minimum=&MakeCustomFloatFromBool(Bool=$User.IsFrontEnd, True Float=0.3, False Float=0.6), Maximum=1)
    )
    Scale RGB=&SelectVectorFromArray(
     Array Sampling Mode=::NewEnumerator0
     Direct Array Index=?
     Vector Selection Array=DataInterface() {propertyValues {FloatData [<1 1 1>, <2 2 2>]}}
     Random Seed=0
    )
    ScaleA=true
    ScaleRGB=true
   )
   &CameraOffset(Camera Offset Amount=-100) {module=CameraOffset001}
   &ScaleColor(
    Scale Alpha=&MakeCustomFloatFromBool(
     Bool=&SetBoolByIntComparison(A=$Particles.FlickerMask, B=0, Comparison Type=::NewEnumerator0)
     True Float=&Waveform(
      "Waveform [1]"=::NewEnumerator0
      "[1] Frequency"=&RandomRangeFloat(Minimum=1, Maximum=1.5)
      "[1] Amplitude Min/Max"=<0.5 1>
     )
     False Float=1
    )
    ScaleA=true
    ScaleRGB=true
   ) {module=ScaleColor001}
   &ScaleColor(
    Scale RGB=&MakeVectorFromLinearColorRGB(Color=&HueShiftLinearColor(
      Hue=&FloatFromCurve(FloatCurve=DataInterface() {propertyValues {Curve {keys [{tangentMode=RCTM_User, time=0, value=0}, {tangentMode=RCTM_User, time=1, value=1}]}}})
      Saturation=0
      Value=1
      Linear Color=$Particles.Color
     ))
    ScaleA=true
    ScaleRGB=true
   ) {module=ScaleColor002}
   &InterpolateOverTime(
    Rate Of Change:NiagaraFloat=&RemapRange(
     Input Value To Remap=&VectorLength(Vector=$Particles.SkeletalMeshLocation001.SampledVelocity)
     Input Min=0
     Input Max=770
     Output Min=1
     Output Max=30
    )
    Target Position Value:NiagaraPosition=$Particles.ShapeLocation001.ShapePosition
    Variable Type=::NewEnumerator6
   ) {module=InterpolateOverTime001}
   /SetVariables(Particles.Position=$Particles.InterpolateOverTime001.InterpolateOverTime_Position.Moving Average) {isSetParameters=true}
  ]
  rendererClasses [&NiagaraSpriteRendererProperties]
  renderers [{rendererClass=&NiagaraSpriteRendererProperties}]
 }
 Dust {
  emitterUpdateScript [
   &EmitterState()
   &SpawnRate(SpawnRate=12, Use Spawn Probability=false)
  ]
  particleSpawnScript [
   &PartitionParticles(Particles Per Partition=4, Partition Mode:ENiagara_PartitionMode=::NewEnumerator1)
   &InitializeParticle(
    Color=&Multiply_LinearColorByFloat(
     Linear Color=&SelectLinearColorFromArray(
      Array Sampling Mode=::NewEnumerator0
      Direct Array Index=?
      Color Selection Array=DataInterface() {propertyValues {ColorData [<rgba 1 0.4392 0 1>, <rgba 0 1 0.3399 1>, <rgba 0 0.3405 1 1>]}}
      Random Seed=0
     )
     Float=&FNTimeOfDayScalar(
      Night Scale=1
      Morning Scale=1.5
      Day Scale=&MakeCustomFloatFromBool(Bool=$User.IsFrontEnd, True Float=1, False Float=1)
      Evening Scale=1
     )
    )
    Lifetime Max=1
    Lifetime Mode=::NewEnumerator1
    Sprite Rotation Mode=::NewEnumerator2
    Sprite Size Mode=::NewEnumerator3
    Uniform Sprite Size=&Multiply_Float(
     A=25
     B=&Multiply_Float(
      A=&Lerp_Float(A=1, B=1, Alpha=$Particles.PartitionParticles.Normalized Index In Partition)
      B=&MakeCustomFloatFromBool(Bool=$User.IsLargeBody, True Float=1.3, False Float=1)
     )
    )
    UsePositionOffset=false
    Write Material Random=true
   )
   /SetVariables(Particles.VisibilityTag=&MakeCustomIntFromBool(Bool=$User.IsFrontEnd, True INT=1, False INT=0)) {isSetParameters=true}
  ]
  particleUpdateScript [
   &ParticleState()
   &SpriteRotationRate(Rotation Rate=30)
   &SkeletalMeshLocation(
    Bone / Socket Index=&SelectIntFromArray(Direct Array Index=$Particles.PartitionParticles.Partition, Int Selection Array=DataInterface() {propertyValues {IntData [1,2]}})
    Bone Sampling Mode=::NewEnumerator4
    MaskSpawnGroup=false
    Orientation Sampling=::NewEnumerator1
    Position Sampling=::NewEnumerator1
    Sampled Position Offset=&Lerp_Vector(
     A=&SelectVectorFromArray(
      Direct Array Index=$Particles.PartitionParticles.Partition
      Vector Selection Array=DataInterface() {propertyValues {FloatData [<-5 0 3>, <5 0 -3>]}}
     )
     B=&SelectVectorFromArray(
      Direct Array Index=$Particles.PartitionParticles.Partition
      Vector Selection Array=DataInterface() {propertyValues {FloatData [<27 0 3>, <-27 0 -3>]}}
     )
     Alpha=$Particles.PartitionParticles.Normalized Index In Partition
    )
    Skeletal Mesh=$System.Skeletal Mesh
   )
   &DynamicMaterialParameters(
    Index 0 Param 1=&FloatFromCurve(FloatCurve=DataInterface() {propertyValues {Curve {keys [<at 0 0 RCIM_Cubic>,<at 1.0011 0.2>]}}})
    Index 0 Param 2=&FloatFromCurve(FloatCurve=DataInterface() {propertyValues {Curve {keys [
         {arriveTangent=0.0453, interpMode=RCIM_Cubic, tangentMode=RCTM_Break, time=0, value=-1}
         {arriveTangent=0.0604, interpMode=RCIM_Cubic, leaveTangent=0.0604, tangentMode=RCTM_User, time=1, value=-1}
        ]}}})
    Index 0 Param 3=&FloatFromCurve(FloatCurve=DataInterface() {propertyValues {Curve {keys [{interpMode=RCIM_Cubic, leaveTangent=4.6779, time=0.2851, value=0.4}]}}})
    Index 1 Param 1=&FloatFromCurve(
     FloatCurve=DataInterface() {propertyValues {Curve {keys [<at 0 0>,<at 0.25 1>]}}}
     Scale Curve=&RandomRangeFloat(Minimum=2, Maximum=1)
    )
    Index 1 Param 2=&FloatFromCurve(
     FloatCurve=DataInterface() {propertyValues {Curve {keys [
         {interpMode=RCIM_Cubic, leaveTangent=3.7422, tangentMode=RCTM_Break, time=0, value=0.2}
         {interpMode=RCIM_Cubic, leaveTangent=0.9066, time=1, value=2}
        ]}}}
     Scale Curve=&RandomRangeFloat(Minimum=0.5, Maximum=1)
    )
    Index 1 Param 3=&FloatFromCurve(FloatCurve=DataInterface() {propertyValues {Curve {keys [
         {interpMode=RCIM_Cubic, leaveTangent=2.8445, tangentMode=RCTM_Break, time=0, value=0.2}
         {interpMode=RCIM_Cubic, leaveTangent=63.2978, time=0.25, value=1}
        ]}}})
    Index 1 Param 4=&FloatFromCurve(
     FloatCurve=DataInterface() {propertyValues {Curve {keys [<at 0 1>,<at 1 0.8>]}}}
     Scale Curve=&RandomRangeFloat(Minimum=0.5, Maximum=1)
    )
    Index 2 Param 1=&FloatFromCurve(FloatCurve=DataInterface() {propertyValues {Curve {keys [{interpMode=RCIM_Cubic, leaveTangent=3.2309, tangentMode=RCTM_Break, time=0.1, value=0}, <at 1 1>]}}})
    Index 2 Param 2=$Particles.MaterialRandom
    Index 2 Param 3=&FloatFromCurve(FloatCurve=DataInterface() {propertyValues {Curve {keys [
         {interpMode=RCIM_Cubic, tangentMode=RCTM_Break, time=0, value=1}
         {interpMode=RCIM_Cubic, leaveTangent=11.1233, tangentMode=RCTM_Break, time=1, value=2}
        ]}}})
    Index 2 Param 4=0.75
    Module.Param0WriteEnabled=true
    Module.Param0WriteEnabled1=false
    Module.Param0WriteEnabled2=true
    Module.Param1WriteEnabled=false
    Module.Param1WriteEnabled1=true
    Module.Param1WriteEnabled2=true
    Module.Param2WriteEnabled=false
    Module.Param2WriteEnabled1=true
    Module.Param2WriteEnabled2=true
    Module.Param3WriteEnabled=true
    Module.Param3WriteEnabled1=true
    Module.Param3WriteEnabled2=true
    Write Parameter Index 1=true
    Write Parameter Index 2=true
   )
   &ScaleColor(
    Scale Alpha=&FloatFromCurve(
     FloatCurve=DataInterface() {propertyValues {Curve {keys [
         {interpMode=RCIM_Cubic, leaveTangent=2.1138, tangentMode=RCTM_Break, time=0, value=0}
         <at 0.4 1 RCIM_Cubic>
         {arriveTangent=-3.4143, interpMode=RCIM_Cubic, leaveTangent=-3.4143, tangentMode=RCTM_User, time=1, value=0}
        ]}}}
     Scale Curve=&MakeCustomFloatFromBool(Bool=$User.IsFrontEnd, True Float=0.5, False Float=1.33)
    )
    ScaleA=true
    ScaleRGB=true
   ) {module=ScaleColor003}
   &GravityForce(Gravity=&RandomRangeVector(Minimum=<0 0 15>, Maximum=<0 1 25>))
   &SolveForcesAndVelocity(Clamp Velocity=false, Limit Acceleration=false)
  ]
  rendererClasses [&NiagaraSpriteRendererProperties]
  renderers [
   {rendererClass=&NiagaraSpriteRendererProperties}
   {rendererClass=&NiagaraSpriteRendererProperties, rendererIndex=1}
  ]
 }
 Glow {
  emitterUpdateScript [
   &EmitterState()
   &SpawnBurst_Instantaneous(Spawn Count=7, Use Spawn Probability=false, UseLoopCountLimit=true)
  ]
  particleSpawnScript [
   &PartitionParticles(Number Of Partitions=7)
   &InitializeParticle(
    Color=&Multiply_LinearColorByFloat(
     Linear Color=<rgba 1 0.7882 0.4549 1>
     Float=&FNTimeOfDayScalar(
      Night Scale=0.9
      Morning Scale=1.1
      Day Scale=&MakeCustomFloatFromBool(Bool=$User.IsFrontEnd, True Float=1, False Float=1)
      Evening Scale=1
     )
    )
    Lifetime=3
    Sprite Rotation Mode=::NewEnumerator0
    Sprite Size Mode=::NewEnumerator3
    Uniform Sprite Size=&Multiply_Float(
     A=&SelectFloatFromArray(
      Direct Array Index=&SpawnGroup()
      Float Selection Array=DataInterface() {propertyValues {FloatData [40,40,20,20,30,30]}}
     )
     B=&MakeCustomFloatFromBool(Bool=$User.IsLargeBody, True Float=1.5, False Float=1)
    )
    UsePositionOffset=false
    Write Material Random=true
   ) {module=InitializeParticle001}
   /SetVariables(Particles.glowScale:Vector2f=&Vector2DFromFloat(Value=&Multiply_Float(
      A=&SelectFloatFromArray(
       Direct Array Index=$Particles.PartitionParticles.Partition
       Float Selection Array=DataInterface() {propertyValues {FloatData [100,70,70,70,70,40,40]}}
      )
      B=&MakeCustomFloatFromBool(Bool=$User.IsLargeBody, True Float=1.2, False Float=1)
     ))) {isSetParameters=true}
   /SetVariables(Particles.glowCameraOffset=-10) {isSetParameters=true}
   /SetVariables() {isSetParameters=true}
  ]
  particleUpdateScript [
   &ParticleState(
    Kill Particles When Lifetime Has Elapsed=false
    Let Infinitely Lived Particles Die When Emitter Deactivates=true
    Loop Particles Lifetime=true
   )
   &CameraOffset(Camera Offset Amount=5)
   &SpriteRotationRate(Rotation Rate=&RandomRangeFloat(Minimum=10, Maximum=20))
   &SkeletalMeshLocation(
    Bone / Socket Index=&SelectIntFromArray(Direct Array Index=$Particles.PartitionParticles.Partition, Int Selection Array=DataInterface() {propertyValues {IntData [0,1,2,3,4,5,6]}})
    Bone Sampling Mode=::NewEnumerator4
    MaskSpawnGroup=false
    Orientation Sampling=::NewEnumerator1
    Sampled Position Offset Coordinate Space:ENiagaraSkelMeshTransforms=::NewEnumerator1
    Skeletal Mesh=$System.Skeletal Mesh
   ) {module=SkeletalMeshLocation003}
   &DynamicMaterialParameters(
    Index 0 Param 1=&SelectFloatFromArray(Direct Array Index=$Particles.PartitionParticles.Partition, Float Selection Array=DataInterface() {propertyValues {FloatData [5,5,5,5,3,3]}})
    Index 0 Param 3=1
    Module.Param0WriteEnabled=true
    Module.Param1WriteEnabled=true
    Module.Param2WriteEnabled=true
    Module.Param3WriteEnabled=true
   )
   /SetVariables(Particles.glowColor:LinearColor=&MakeLinearColorFromVectorAndFloat(
     "Vector (RGB)"=&MakeVectorFromLinearColorRGB(Color=<rgba 1 0.6376 0.107 1>)
     "Float (Alpha)"=&FloatFromCurve(FloatCurve=DataInterface() {propertyValues {Curve {keys [<at 0 1>,<at 1 1>]}}}, Scale Curve=0.15)
    )) {isSetParameters=true}
  ]
  rendererClasses [&NiagaraSpriteRendererProperties]
  renderers [
   {rendererClass=&NiagaraSpriteRendererProperties}
   {rendererClass=&NiagaraSpriteRendererProperties, rendererIndex=1}
  ]
 }
 Rays {
  emitterUpdateScript [
   &EmitterState()
   &SpawnBurst_Instantaneous(Use Spawn Probability=false, UseLoopCountLimit=true)
  ]
  particleSpawnScript [
   &InitializeParticle(Mesh Scale Mode=::NewEnumerator3, Mesh Uniform Scale=1.5, UsePositionOffset=false, Write Material Random=true)
   &DynamicMaterialParameters(
    Index 0 Param 1=&RandomRangeFloat(Minimum=0, Maximum=10)
    Index 0 Param 2=&RandomRangeFloat(Minimum=1, Maximum=2)
    Index 0 Param 3=&RandomRangeFloat(Minimum=0.8, Maximum=1.6)
    Module.Index0Param0Write=true
    Module.Index0Param1Write=true
    Module.Index0Param2Write=true
    Module.Index0Param3Write=false
   )
   &SkeletalMeshLocation(
    Bone / Socket Index=7
    Bone Sampling Mode=::NewEnumerator4
    MaskSpawnGroup=false
    Orientation Sampling=::NewEnumerator1
    Skeletal Mesh=$System.Skeletal Mesh
   )
   &CameraOffset(Camera Offset Amount=-50)
   &InitialMeshOrientation(Mesh Orientation Mode=::NewEnumerator2, Rotation=<0.5 0 0>, Use Rotation Vector=true)
  ]
  particleUpdateScript [
   &ParticleState(Kill Particles When Lifetime Has Elapsed=false)
   &SkeletalMeshLocation(
    Bone / Socket Index=7
    Bone Sampling Mode=::NewEnumerator4
    MaskSpawnGroup=false
    Orientation Sampling=::NewEnumerator1
    Skeletal Mesh=$System.Skeletal Mesh
   ) {module=SkeletalMeshLocation001}
   &ScaleColor(
    Scale Alpha=&MakeCustomFloatFromBool(Bool=$User.IsFrontEnd, True Float=0.15, False Float=0.4)
    Scale RGB=&Multiply_VectorByFloat(
     Vector=&MakeVectorFromLinearColorRGB(Color=&ColorFromCurve(
       ColorCurve=DataInterface() {propertyValues {
         AlphaCurve {keys []}
         BlueCurve {keys [<at 0 0 RCIM_Cubic>,<at 0.33 1>,<at 0.66 0.3399>,<at 1 0 RCIM_Cubic>]}
         GreenCurve {keys [<at 0 0.4392 RCIM_Cubic>,<at 0.33 0.3405>,<at 0.66 1>,<at 1 0.4392 RCIM_Cubic>]}
         RedCurve {keys [<at 0 1 RCIM_Cubic>,<at 0.33 0>,<at 0.66 0>,<at 1 1 RCIM_Cubic>]}
        }}
       CurveIndex=&Frac_Float(Float=&Multiply_Float(A=$Engine.Time, B=1))
      ))
     Float=&FNTimeOfDayScalar(Night Scale=0.9, Morning Scale=1, Evening Scale=1)
    )
    ScaleA=true
    ScaleRGB=true
   ) {module=ScaleColor001}
  ]
  rendererClasses [&NiagaraMeshRendererProperties]
  renderers [{rendererClass=&NiagaraMeshRendererProperties, rendererIndex=0}]
 }
}
summary {
 systemName=NS_Aura_Bokeh
 emitters [
  {emitterName=Glow, rendererClasses [&NiagaraSpriteRendererProperties]}
  {emitterName=BokehLights, rendererClasses [&NiagaraSpriteRendererProperties]}
  {emitterName=Rays, rendererClasses [&NiagaraMeshRendererProperties]}
  {emitterName=BokehLights001, rendererClasses [&NiagaraSpriteRendererProperties]}
  {emitterName=Dust, rendererClasses [&NiagaraSpriteRendererProperties]}
 ]
 userVariables [
  User.Interpolate Intensity {type=NiagaraFloat, defaultValue=0.1, description=""}
  User.IsFrontEnd {type=NiagaraBool}
  User.IsLargeBody {type=NiagaraBool}
 ]
}
